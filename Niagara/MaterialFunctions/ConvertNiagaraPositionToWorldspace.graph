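# Material function, exposed in the Niagara library category, that converts a Niagara simulation position to world space.
# It takes a SimulationPosition input and a four-component SystemLWCTile input for the system's Large World Coordinates tile.
# It multiplies the tile's RGB by its alpha to form an offset, then adds that to the simulation position.
# The sum is a plain add and multiply, with no dedicated LWC nodes.
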MaterialFunction /Niagara/MaterialFunctions/ConvertNiagaraPositionToWorldspace
graph {
 properties {bExposeToLibrary=true, libraryCategoriesText [Niagara]}
 nodes [
  n1 = FunctionInput() {inputName=SystemLWCTile, inputType=FunctionInput_Vector4, sortPriority=1}
  n2 = ComponentMask(None=@n1) {b=true, g=true, r=true}
  n3 = ComponentMask(None=@n1) {a=true}
  FunctionOutput(None=Add(A=FunctionInput() {inputName=SimulationPosition}, B=Multiply(A=@n2, B=@n3)))
 ]
}
