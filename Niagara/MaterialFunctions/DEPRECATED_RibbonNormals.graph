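# Deprecated Niagara material function that outputs smooth cylindrical world-space normals along ribbons.
# Builds a rounded normal from the inverted, wrapped V texture coordinate and angle-blends it with a green-flipped normal map input.
# Transforms the result by a camera-facing orthogonal basis built from the U tangent and the particle-to-camera vector.
# Feeding the Normal pin directly requires the material's tangent-space normal setting to be off.
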MaterialFunction /Niagara/MaterialFunctions/DEPRECATED_RibbonNormals
graph {
 properties {description=Deprecated: Produces smooth cylindrical normals along Niagara ribbons., libraryCategoriesText [Niagara]}
 nodes [
  n1 = TextureCoordinate()
  n2 = Frac(None=OneMinus(None=MaterialFunctionCall(Float2=@n1) {materialFunction=&BreakOutFloat2Components}.G))
  n3 = DeriveNormalZ(InXY=AppendVector(A=ConstantBiasScale(None=@n2) {bias=-0.5, scale=2}, B=Constant()))
  n4 = Multiply(
   A=FunctionInput() {bUsePreviewValueAsDefault=true, inputName=Normals Map, previewValue=<0 0 1 1>}
   B=Constant3Vector() {constant=<rgba 1 -1 1 0>}
  )
  n5 = Subtract(A=CameraPositionWS(), B=ParticlePositionWS())
  n6 = MaterialFunctionCall() {materialFunction=&TangentBasis}
  n7 = MaterialFunctionCall!(Vector1=Multiply(A=CrossProduct(A=@n5, B=@n6.X U)) {constB=-1}, Vector2=@n6.X U) {materialFunction=/Engine/Functions/Engine_MaterialFunctions02/Math/CreateThirdOrthogonalVector}
  FunctionOutput(None=MaterialFunctionCall(
    VectorToTransform=MaterialFunctionCall(BaseNormal=@n3, AdditionalNormal=@n4) {materialFunction=&BlendAngleCorrectedNormals}.Result
    BasisX=Normalize(VectorInput=@n7.Vector2)
    BasisY=Normalize(VectorInput=@n7.Vector1)
    BasisZ=Normalize(VectorInput=@n7.Vector3)
   ) {materialFunction=&Transform3x3Matrix}.Transformed Vector) {
   description=Uncheck tangent space normals in your base material properties before directly feeding this output into the shader's normals pin.
   outputName=World Space Normals
  }
 ]
}
